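# Sound cue for the checkpoint-reached sound in the Uranium limited-time mode.
# A single wave player feeds the root directly with the Uranium_Checkpoint_Reached_01 wave, with no randomization or mixing nodes.
# It uses the critical in-game UI sound class and overrides concurrency so only one instance plays at a time.
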SoundCue /Game/Sounds/LTM/Uranium/CheckpointReached_Cue
root=WavePlayer_0
nodes [{id=WavePlayer_0, properties {soundWaveAssetPtr=&/Uranium_Checkpoint_Reached_01}}]
properties {
 bOverrideConcurrency=true
 maxDistance=2097152
 soundClassObject=&Fort_UI_InGame_Critical
 concurrencyOverrides {maxCount=1}
}
